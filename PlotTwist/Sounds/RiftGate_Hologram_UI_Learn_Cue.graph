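# Sound cue for the RiftGate hologram UI "Learn" sound, routed to the non-critical in-game UI sound class.
# A single wave player feeds the root with no randomization or mixing, at a slightly boosted volume.
# Overridden attenuation disables spatialization and reverb send while defining a distance falloff.
# Overridden concurrency allows one instance at a time, stopping the oldest when a new one starts.
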SoundCue /PlotTwist/Sounds/RiftGate_Hologram_UI_Learn_Cue
root=WavePlayer_2
nodes [{id=WavePlayer_2, properties {soundWaveAssetPtr=&/RiftGate_Hologram_UI_Learn_01}}]
properties {
 bOverrideAttenuation=true
 bOverrideConcurrency=true
 maxDistance=2000
 soundClassObject=&Fort_UI_InGame_NonCritical
 volumeMultiplier=1.2
 attenuationOverrides {bEnableReverbSend=false, bSpatialize=false, falloffDistance=1200, attenuationShapeExtents {x=800}}
 concurrencyOverrides {maxCount=1, resolutionRule=StopOldest}
}
