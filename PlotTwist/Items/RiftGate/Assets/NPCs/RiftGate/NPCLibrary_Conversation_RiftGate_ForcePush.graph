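# Conversation database for the RiftGate NPC dialogue that offers Force Push training with Obi-Wan.
# The entry tag opens a speech asking whether the player wishes to train in the Force, then branches into two choices.
# "Train with Obi-Wan", the initially selected choice, runs the GainForcePush task node with a learn sound before a blessing line.
# "Leave" is a top-priority back choice that plays a leave sound and ends with a farewell.
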ConversationDatabase /PlotTwist/Items/RiftGate/Assets/NPCs/RiftGate/NPCLibrary_Conversation_RiftGate_ForcePush
entry_points [{tag=Conversation.Entry.RiftGate.ForcePush, destinations [#1]}]
nodes [
 ConversationEntryPointNode_0 {
  id=#1
  class=ConversationEntryPointNode
  outputs [#2]
  properties {entryTag {tagName=Conversation.Entry.RiftGate.ForcePush}}
 }
 Speech_C_1 {
  id=#2
  class=Speech_C
  outputs [#3, #4]
  properties {nodeName="", generalConfig {defaultMessage=Do you wish the train in the ways of the Force to strengthen your powers?}}
 }
 RiftGate_TaskNode_GainForcePowers_C_0 {
  id=#4
  class=RiftGate_TaskNode_GainForcePush_C
  outputs [#5]
  properties {choiceSound=/PlotTwist/Sounds/RiftGate_Hologram_UI_Learn_Cue}
  sub_nodes [ConversationChoiceNode_2 {
    class=ConversationChoiceNode
    properties {defaultChoiceDisplayText=Train with Obi-Wan, choiceTags {gameplayTags [{tagName=Conversation.Choice.InitialSelectedIndex}]}}
   }]
 }
 Speech_C_2 {
  id=#5
  class=Speech_C
  outputs []
  properties {nodeName="", generalConfig {defaultMessage=May the Force be with you.}}
 }
 Continue_C_1 {
  id=#3
  class=Continue_C
  outputs [#6]
  properties {choiceSound=/PlotTwist/Sounds/RiftGate_Hologram_UI_Leave_Cue, nodeName=Exit, softTaskIcon=&Cancel_Normal}
  sub_nodes [ConversationChoiceNode_1 {
    class=ConversationChoiceNode
    properties {
     defaultChoiceDisplayText=Leave
     choiceTags {gameplayTags [{tagName=Conversation.Choice.Category.Back}, {tagName=Conversation.Choice.CustomSortPriority.Highest}]}
    }
   }]
 }
 Speech_C_3 {id=#6, class=Speech_C, outputs [], properties {nodeName="", generalConfig {defaultMessage="Farewell, then."}}}
]
