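# Conversation database for a Rift Gate NPC that offers Force training with Ahsoka Tano, entered through the Conversation.Entry.RiftGate.ForceWht tag.
# An opening speech branches into a pre-selected train choice and a top-priority Leave choice in the Back category, each with its own hologram UI sound.
# Training runs the GainForceWht task node before a closing line, while Leave uses a Continue node to reach a different farewell.
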ConversationDatabase /PlotTwist/Items/RiftGate/Assets/NPCs/RiftGate/NPCLibrary_Conversation_RiftGate_Force_Wht
entry_points [{tag=Conversation.Entry.RiftGate.ForceWht, destinations [#1]}]
nodes [
 ConversationEntryPointNode_0 {
  id=#1
  class=ConversationEntryPointNode
  outputs [#2]
  properties {entryTag {tagName=Conversation.Entry.RiftGate.ForceWht}}
 }
 Speech_C_1 {
  id=#2
  class=Speech_C
  outputs [#3, #4]
  properties {nodeName="", generalConfig {defaultMessage=I am no Jedi.}}
 }
 RiftGate_TaskNode_GainForceWht_C_0 {
  id=#4
  class=RiftGate_TaskNode_GainForceWht_C
  outputs [#5]
  properties {hoverChoiceSound=/PlotTwist/Sounds/RiftGate_Hologram_UI_Learn_Cue}
  sub_nodes [ConversationChoiceNode_4 {
    class=ConversationChoiceNode
    properties {defaultChoiceDisplayText=Train with Ahsoka Tano, choiceTags {gameplayTags [{tagName=Conversation.Choice.InitialSelectedIndex}]}}
   }]
 }
 Speech_C_2 {
  id=#5
  class=Speech_C
  outputs []
  properties {nodeName="", generalConfig {defaultMessage="I like firsts. Good or bad, they're always memorable."}}
 }
 Continue_C_1 {
  id=#3
  class=Continue_C
  outputs [#6]
  properties {choiceSound=/PlotTwist/Sounds/RiftGate_Hologram_UI_Leave_Cue, nodeName=Exit, softTaskIcon=&Cancel_Normal}
  sub_nodes [ConversationChoiceNode_1 {
    class=ConversationChoiceNode
    properties {
     defaultChoiceDisplayText=Leave
     choiceTags {gameplayTags [{tagName=Conversation.Choice.Category.Back}, {tagName=Conversation.Choice.CustomSortPriority.Highest}]}
    }
   }]
 }
 Speech_C_3 {
  id=#6
  class=Speech_C
  outputs []
  properties {nodeName="", generalConfig {defaultMessage=I don't control the wants of others.}}
 }
]
