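# Conversation database for the RiftGate HeavyLamp NPC, a dialogue offering the player training with Darth Vader.
# The tagged entry point plays an opening line, then branches into a train choice, marked as initially selected, and a Leave choice sorted as a back option.
# Training runs the GainHeavyLamp task, gated by an inverted owned-tag requirement on the Falling movement mode, then plays a Dark Side line.
# Leaving plays a sound and a disapproving parting line.
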ConversationDatabase /PlotTwist/Items/RiftGate/Assets/NPCs/RiftGate/NPCLibrary_Conversation_RiftGate_HeavyLamp
entry_points [{tag=Conversation.Entry.RiftGate.HeavyLamp, destinations [#1]}]
nodes [
 ConversationEntryPointNode_0 {
  id=#1
  class=ConversationEntryPointNode
  outputs [#2]
  properties {entryTag {tagName=Conversation.Entry.RiftGate.HeavyLamp}}
 }
 Speech_C_2 {
  id=#2
  class=Speech_C
  outputs [#3, #4]
  properties {nodeName="", generalConfig {defaultMessage="If you could be turned, you would become a powerful ally."}}
 }
 RiftGate_TaskNode_GainHeavyLamp_C_1 {
  id=#4
  class=RiftGate_TaskNode_GainHeavyLamp_C
  outputs [#5]
  sub_nodes [
   ConversationChoiceNode_3 {
    class=ConversationChoiceNode
    properties {defaultChoiceDisplayText=Train with Darth Vader, choiceTags {gameplayTags [{tagName=Conversation.Choice.InitialSelectedIndex}]}}
   }
   RiftGate_ConversationRequirement_CannotHaveTagRequirement_C_1 {
    class=RiftGate_ConversationRequirement_CannotHaveTagRequirement_C
    properties {requirement {class=FortControllerRequirement_HasOwnedTag, properties {bInverseCondition=true, ownedTag {tagName=MovementMode.Falling}}}}
   }
  ]
 }
 Speech_C_0 {
  id=#5
  class=Speech_C
  outputs []
  properties {nodeName="", generalConfig {defaultMessage=Give yourself to the Dark Side.}}
 }
 Continue_C_1 {
  id=#3
  class=Continue_C
  outputs [#6]
  properties {choiceSound=/PlotTwist/Sounds/RiftGate_Hologram_UI_Leave_Cue, nodeName=Exit, softTaskIcon=&Cancel_Normal}
  sub_nodes [ConversationChoiceNode_1 {
    class=ConversationChoiceNode
    properties {
     defaultChoiceDisplayText=Leave
     choiceTags {gameplayTags [{tagName=Conversation.Choice.Category.Back}, {tagName=Conversation.Choice.CustomSortPriority.Highest}]}
    }
   }]
 }
 Speech_C_3 {
  id=#6
  class=Speech_C
  outputs []
  properties {nodeName="", generalConfig {defaultMessage=I find your lack of faith disturbing.}}
 }
]
