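# A Rift Gate NPC conversation database in which Mace Windu offers to train the player in Thunder Barn Pull.
# The entry tag leads to an opening line, then two choices.
# The initially selected "Train with Mace Windu" choice runs a GainThunderBarnPull task node with a learn sound, then a parting line.
# "Leave" is a Back-category, top-priority exit with a cancel icon and leave sound, leading to a dismissive line.
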ConversationDatabase /PlotTwist/Items/RiftGate/Assets/NPCs/RiftGate/NPCLibrary_Conversation_RiftGate_ThunderBarnPull
entry_points [{tag=Conversation.Entry.RiftGate.ThunderBarn.Pull, destinations [#1]}]
nodes [
 ConversationEntryPointNode_7 {
  id=#1
  class=ConversationEntryPointNode
  outputs [#2]
  properties {entryTag {tagName=Conversation.Entry.RiftGate.ThunderBarn.Pull}}
 }
 Speech_C_2 {
  id=#2
  class=Speech_C
  outputs [#3, #4]
  properties {nodeName="", generalConfig {defaultMessage="Do not let fear cloud your judgment, for only through clarity can we find wisdom."}}
 }
 RiftGate_TaskNode_GainThunderBarnPull_C_1 {
  id=#4
  class=RiftGate_TaskNode_GainThunderBarnPull_C
  outputs [#5]
  properties {choiceSound=/PlotTwist/Sounds/RiftGate_Hologram_UI_Learn_Cue}
  sub_nodes [ConversationChoiceNode_3 {
    class=ConversationChoiceNode
    properties {defaultChoiceDisplayText=Train with Mace Windu, choiceTags {gameplayTags [{tagName=Conversation.Choice.InitialSelectedIndex}]}}
   }]
 }
 Speech_C_1 {
  id=#5
  class=Speech_C
  outputs []
  properties {nodeName="", generalConfig {defaultMessage=We must move quickly if the Jedi Order is to survive.}}
 }
 Continue_C_1 {
  id=#3
  class=Continue_C
  outputs [#6]
  properties {choiceSound=/PlotTwist/Sounds/RiftGate_Hologram_UI_Leave_Cue, nodeName=Exit, softTaskIcon=&Cancel_Normal}
  sub_nodes [ConversationChoiceNode_1 {
    class=ConversationChoiceNode
    properties {
     defaultChoiceDisplayText=Leave
     choiceTags {gameplayTags [{tagName=Conversation.Choice.Category.Back}, {tagName=Conversation.Choice.CustomSortPriority.Highest}]}
    }
   }]
 }
 Speech_C_3 {
  id=#6
  class=Speech_C
  outputs []
  properties {nodeName="", generalConfig {defaultMessage=A shame you can't feel the Force. You might have made a fine Jedi.}}
 }
]
