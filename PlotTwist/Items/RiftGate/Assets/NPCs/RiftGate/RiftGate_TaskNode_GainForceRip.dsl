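; RiftGate_TaskNode_GainForceRip is a conversation task node that grants Force powers when a dialogue with a Rift Gate NPC reaches it.
; It takes the participant NPC actor from the conversation context, casts it to the Athena PlotTwist RiftGate parent blueprint and calls GainForcePowers on it.
; It then returns the result of the parent task node implementation.
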(fn ExecuteTaskNode
  (params (Context :type Conversation%20Context%20Structure%20%28by%20ref%29))
  (bind _returnvalue (Class|FortConversationContextLibrary|GetParticipantActorNPC Context))
  (bind _asb_athena_plot_twist_rift_gate_parent (Utilities|Casting|CastToB_Athena_PlotTwist_RiftGate_Parent _returnvalue))
  (Class|BAthenaPlotTwistRiftGateParent|GainForcePowers _asb_athena_plot_twist_rift_gate_parent)
  (bind _returnvalue_2 (|CallParentFunction (Variables|GetContext)))
  (return _returnvalue_2))
