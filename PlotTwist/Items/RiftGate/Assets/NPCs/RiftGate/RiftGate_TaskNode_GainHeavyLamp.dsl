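; Conversation task node that grants heavy lamp training when executed in a PlotTwist RiftGate NPC conversation.
; It takes the conversation's NPC participant actor, casts it to the Heavy Lamp convo actor and calls GrantTraining on it.
; The parent task's result is then returned.
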(fn ExecuteTaskNode
  (params (Context :type Conversation%20Context%20Structure%20%28by%20ref%29))
  (bind _returnvalue (Class|FortConversationContextLibrary|GetParticipantActorNPC Context))
  (bind _asbga_athena_plot_twist_heavy_lamp_convo (Utilities|Casting|CastToBGA_Athena_PlotTwist_HeavyLamp_Convo _returnvalue)
    (:then
      (Class|BGAAthenaPlotTwistHeavyLampConvo|GrantTraining _asbga_athena_plot_twist_heavy_lamp_convo))
    (:CastFailed
      _))
  (bind _returnvalue_2 (|CallParentFunction (Variables|GetContext)))
  (return _returnvalue_2))
